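; Actor component that creates a shroud zone around its owner, limiting visibility for autofire aim-assist targeting.
; When enabled by console variable or in editor, it adds an overlap sphere, which external activation sizes and sets to block the Visibility channel.
; When the locally controlled player enters, the sphere ignores Visibility and the aim-assist target manager gets a custom range for the owner.
; Exiting restores blocking and clears that range; deactivation disables collision.
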(event EventBeginPlay
  (bind _bactivate (CallFunction|ShouldActivateComponent))
  (if _bactivate
    (CallFunction|CreateVisibilityBlockers)
    (Default|InitializeOverlaps)))

(event EventEndPlay
  (params (EndPlayReason :type EEndPlayReason%20Enum :default "Destroyed"))
  (Default|DisableAllBlockers))

(event EventTick
  (params (DeltaSeconds :type Float%20%28single-precision%29)))

(event Custom|AddPlayerToShroudZone
  (params (OverlappedComponent :type Primitive%20Component%20Object%20Reference) (OtherActor :type Actor%20Object%20Reference) (OtherComp :type Primitive%20Component%20Object%20Reference) (OtherBodyIndex :type Integer) (bFromSweep :type Boolean) (SweepResult :type Hit%20Result%20Structure%20%28by%20ref%29))
  (bind _asfort_player_pawn (Utilities|Casting|CastToFortPlayerPawn OtherActor))
  (if (Pawn|IsLocallyControlled _asfort_player_pawn)
    (Default|DisableSphereVisibilityBlocker)
    (CallFunction|SetVisibilityRangeModifier)))

(event Custom|RemovePlayerFromShroudZone
  (params (OverlappedComponent :type Primitive%20Component%20Object%20Reference) (OtherActor :type Actor%20Object%20Reference) (OtherComp :type Primitive%20Component%20Object%20Reference) (OtherBodyIndex :type Integer))
  (bind _asfort_player_pawn (Utilities|Casting|CastToFortPlayerPawn OtherActor))
  (if (Pawn|IsLocallyControlled _asfort_player_pawn)
    (Default|EnableSphereVisibilityBlocker)
    (CallFunction|ClearVisibilityRangeModifier)))

(fn ActivateShroudDetectors
  (params (ShourdRadius :type Float%20%28double-precision%29))
  (bind _bactivate (CallFunction|ShouldActivateComponent))
  (if _bactivate
    (Variables|Default|SetShroudRadius ShourdRadius)
    (Default|EnableSphereVisibilityBlocker)
    (return _bactivate)))

(fn ClearVisibilityRangeModifier
  (bind _returnvalue (GameInstanceSubsystems|GetFortAimAssist2D_TargetManager))
  (Utilities|IsValid _returnvalue
    (:Is%20Valid
      (if (Variables|Default|GetbIsLocalPlayerVisibilityShrouded)
        (Class|FortAimAssist2DTargetManager|ClearCustomTargetRange _returnvalue (Components|GetOwner))
        (Variables|Default|SetbIsLocalPlayerVisibilityShrouded false)))))

(fn CreateVisibilityBlockers
  ; Sphere Visibility Blocker
  (bind _returnvalue (Game|AddComponentbyClass :self (Components|GetOwner) :Class "/Script/Engine.SphereComponent"))
  (bind _output_get (Variables|Default|SetSphereVisibilityVolume _returnvalue))
  (Variables|/Script/Engine.PrimitiveComponent|SetbGenerateOverlapEvents true _output_get)
  (Collision|SetCollisionProfileName _output_get "OverlapOnlyPawn" false :!class /Script/Engine.PrimitiveComponent)
  (Components|Sphere|SetSphereRadius _output_get 0.0 false))

(fn DeactivateShroudDetectors
  (Default|DisableAllBlockers))

(fn DisableAllBlockers
  (if (not (Variables|Default|GetbFullDisabled))
    (Collision|SetCollisionEnabled (Variables|Default|GetSphereVisibilityVolume) "NoCollision")
    (CallFunction|ClearVisibilityRangeModifier)
    (Variables|Default|SetbFullDisabled true)))

(fn DisableSphereVisibilityBlocker
  (Collision|SetCollisionResponsetoChannel (Variables|Default|GetSphereVisibilityVolume) "ECC_Visibility" "ECR_Ignore"))

(fn DrawDebugHelpers
  (Rendering|Debug|DrawDebugSphere (Transformation|GetWorldLocation (Variables|Default|GetSphereVisibilityVolume)) (+ (Variables|/Script/Engine.SphereComponent|GetSphereRadius (Variables|Default|GetSphereVisibilityVolume)) (Variables|Default|GetMaxShroudedVisibilityRange)) 12 "(R=1.000000,G=0.936853,B=0.000000,A=0.669565)" 0.1)
  (switch Utilities|FlowControl|Switch|SwitchonECollisionResponse (Collision|GetCollisionResponsetoChannel (Variables|Default|GetSphereVisibilityVolume) "ECC_Visibility")
    (:ECR_Ignore
      (Rendering|Debug|DrawDebugSphere (Transformation|GetWorldLocation (Variables|Default|GetSphereVisibilityVolume)) (Variables|/Script/Engine.SphereComponent|GetSphereRadius (Variables|Default|GetSphereVisibilityVolume)) 12 "(R=1.000000,G=0.000000,B=0.089592,A=1.000000)" 0.1))
    (:ECR_Block
      (Rendering|Debug|DrawDebugSphere (Transformation|GetWorldLocation (Variables|Default|GetSphereVisibilityVolume)) (Variables|/Script/Engine.SphereComponent|GetSphereRadius (Variables|Default|GetSphereVisibilityVolume)) 12 "(R=0.153863,G=1.000000,B=0.000000,A=1.000000)" 0.1))))

(fn EnableSphereVisibilityBlocker
  (bind _spherevisibilityvolume (Variables|Default|GetSphereVisibilityVolume))
  (LocalVariables|SetBlockerRadius (* (Variables|Default|GetShroudRadius) (Variables|Default|GetSphereBlockerAdjustmentMultiplier)))
  (Collision|SetCollisionResponsetoChannel _spherevisibilityvolume "ECC_Visibility" "ECR_Block")
  (Collision|SetCollisionEnabled _spherevisibilityvolume "QueryOnly")
  (Components|Sphere|SetSphereRadius _spherevisibilityvolume (LocalVariables|GetBlockerRadius) true))

(fn InitializeOverlaps
  (Collision|BindEventtoOnComponentEndOverlap (Variables|Default|GetSphereVisibilityVolume) (EventDispatchers|CreateEvent :!event RemovePlayerFromShroudZone))
  (Collision|BindEventtoOnComponentBeginOverlap (Variables|Default|GetSphereVisibilityVolume) (EventDispatchers|CreateEvent :!event AddPlayerToShroudZone)))

(fn SetVisibilityRangeModifier
  (bind _returnvalue (GameInstanceSubsystems|GetFortAimAssist2D_TargetManager))
  (Utilities|IsValid _returnvalue
    (:Is%20Valid
      (Class|FortAimAssist2DTargetManager|SetCustomTargetRange _returnvalue (Components|GetOwner) (Variables|Default|GetMaxShroudedVisibilityRange))
      (Variables|Default|SetbIsLocalPlayerVisibilityShrouded true))))

(fn ShouldActivateComponent
  (bind _returnvalue (Development|GetConsoleVariableBoolValue "Touch.EnableAutofireShroudComponent"))
  (return (or (Editor|InEditororPIE) _returnvalue)))
