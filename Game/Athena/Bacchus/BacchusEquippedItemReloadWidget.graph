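# Bacchus HUD widget showing the equipped item's ammo, resource or utility count, with a pop-up for item name and durability.
# Construct hands bound widgets and intro/outro animations to the native base class, while PreConstruct adjusts the name box height for Korean.
# It registers as a per-player widget host under the EquippedItemInfo slot tag, adding or removing extension widgets below the name.
# Hiding reserve ammo collapses surplus counts and the utility icon and divider.
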WidgetBlueprint /Game/Athena/Bacchus/BacchusEquippedItemReloadWidget
animations [
 NameIntro {
  binding_count=1
  length_seconds=0.35
  object=&self:NameIntro
  sequence {bindings [SwitchPopUp {
     id=#1
     possessed_class=VerticalBox
     tracks [
      {
       type=MovieSceneFloatTrack
       display_name=Render Opacity
       sections [{type=MovieSceneFloatSection, end_frame=7, start_frame=0, channels [None {type=MovieSceneScriptingFloatChannel, keys [[0,0],[7,1]]}]}]
      }
      {
       type=MovieScene2DTransformTrack
       sections [{
         type=MovieScene2DTransformSection
         end_frame=7
         start_frame=0
         channels [Translation.X {type=MovieSceneScriptingFloatChannel, keys [[0,63.2905],[7,0]]}]
        }]
      }
     ]
    }]}
 }
 NameOutro {
  binding_count=1
  length_seconds=2
  object=&self:NameOutro
  sequence {bindings [SwitchPopUp {
     id=#2
     possessed_class=VerticalBox
     tracks [
      {
       type=MovieSceneFloatTrack
       display_name=Render Opacity
       sections [{type=MovieSceneFloatSection, end_frame=40, start_frame=0, channels [None {type=MovieSceneScriptingFloatChannel, keys [[0,1],[40,0.5]]}]}]
      }
      {
       type=MovieScene2DTransformTrack
       sections [{type=MovieScene2DTransformSection, end_frame=40, start_frame=0, channels [Translation.X {type=MovieSceneScriptingFloatChannel, const [0,1]}]}]
      }
     ]
    }]}
 }
]
graphs [
 AddContentWidget {nodes [
   n1 = AddContentWidget!() -> (then, WidgetSlotTag:Gameplay Tag Structure, ContentWidget:Widget Object Reference) {is=title K2Node_FunctionEntry}
   > Branch(Condition=MatchesTag!(
     TagOne:Gameplay Tag Structure=@n1
     TagTwo:Gameplay Tag Structure=Get ItemInfoSlotTag() -> (ItemInfoSlotTag:Gameplay Tag Structure)
     bExactMatch:bool=true
    ) -> (ReturnValue:bool))
   > n2 = AddChildToHorizontalBox!(
    execute
    self:Horizontal Box Object Reference=Get HBox_ExtensionWidgets() -> (HBox_ExtensionWidgets:Horizontal Box Object Reference)
    Content:Widget Object Reference=Get ContentWidget() -> (ContentWidget:Widget Object Reference)
   ) -> (then, ReturnValue:Horizontal Box Slot Object Reference)
   > SetHorizontalAlignment(self:Horizontal Box Slot Object Reference=@n2, InHorizontalAlignment=HAlign_Fill)
   > SetVerticalAlignment(self:Horizontal Box Slot Object Reference=@n2, InVerticalAlignment=VAlign_Fill)
  ]}
 EventGraph {nodes [
   Event PreConstruct()
   > SetHeightOverride!(execute, self:Self Object Reference) -> (then)
   Event Construct()
   > Set Split(Split:Image Object Reference=Get Split1() -> (Split1:Image Object Reference)) -> (Output_Get:Image Object Reference)
   > Set UtilityAmmoIcon(UtilityAmmoIcon:Image Object Reference=Get Image_UtilityAmmoIcon() -> (Image_UtilityAmmoIcon:Image Object Reference)) -> (Output_Get:Image Object Reference)
   > Set SwitchPopUpBox(SwitchPopUpBox:Vertical Box Object Reference=Get SwitchPopUp() -> (SwitchPopUp:Vertical Box Object Reference)) -> (Output_Get:Vertical Box Object Reference)
   > Set SwitchPopUpBoxTextBlock(SwitchPopUpBoxTextBlock:Common Text Object Reference=Get SwitchPopUpText() -> (SwitchPopUpText:Common Text Object Reference)) -> (Output_Get:Common Text Object Reference)
   > Set NameIntroAnimation(NameIntroAnimation:Widget Animation Object Reference=Get NameIntro() -> (NameIntro:Widget Animation Object Reference)) -> (Output_Get:Widget Animation Object Reference)
   > Set NameOutroAnimation(NameOutroAnimation:Widget Animation Object Reference=Get NameOutro() -> (NameOutro:Widget Animation Object Reference)) -> (Output_Get:Widget Animation Object Reference)
   > RegisterWidgetSlots()
   Event Handle Resource Type Changed!() -> (OutputDelegate, then, Item:Fort Resource Item Definition Object Reference, ResourceCount:int=0) {is=title K2Node_Event}
   > Native_UpdateWidgetColor!(execute, self:Bacchus Equipped Item Reload Widget Base Object Reference, bHasAmmo:bool=true) -> (then)
   Event Destruct()
   > UnregisterWidgetSlots()
   n1 = Event Add Content To Host Widget!() -> (
    OutputDelegate
    then
    WidgetHostTag:"Gameplay Tag Structure (by ref)"
    ContentWidget:Widget Object Reference
    Padding:"Margin Structure (by ref)"
    bFill:bool
   ) {is=title K2Node_Event}
   > AddContentWidget!(execute, self, WidgetSlotTag:Gameplay Tag Structure=@n1.WidgetHostTag, ContentWidget:Widget Object Reference=@n1.ContentWidget) -> (then) {is=title K2Node_CallFunction}
   n2 = Event BP_HandleHideReserveAmmoChanged!() -> (OutputDelegate, then, bHideReserveAmmo:bool) {is=title K2Node_Event}
   > SetVisibility(
    self=[
     Get Overlay_UtilityAmmoIcon() -> (Overlay_UtilityAmmoIcon:Overlay Object Reference)
     Get Image_UtilitySplit() -> (Image_UtilitySplit:Image Object Reference)
     Get Text_SurplusAmmoCount() -> (Text_SurplusAmmoCount:Common Text Object Reference)
     Get Text_SurplusUtilityAmmoCount() -> (Text_SurplusUtilityAmmoCount:Common Text Object Reference)
    ]
    InVisibility=Select(
     Option 0=Get WeaponEquipVisibility() -> (WeaponEquipVisibility:ESlateVisibility Enum=Visible)
     Option 1=Collapsed
     Index=@n2.bHideReserveAmmo
    )
   )
  ]}
 RegisterWidgetSlots {nodes [
   RegisterWidgetSlots() {is=title K2Node_FunctionEntry}
   > RegisterWidgetHost!(
    execute
    WidgetHost:Widget Object Reference
    WidgetHostTag:"Gameplay Tag Structure (by ref)"=Get ItemInfoSlotTag()
    IsPlayerSpecific:bool=true
   ) -> (then)
  ]}
 RemoveContentFromWidgetHost {nodes [
   n1 = Remove Content From Host Widget!() -> (
    then
    WidgetHostTag:"Gameplay Tag Structure (by ref)"
    ContentWidget:Widget Object Reference
    RemovalType:EWidgetSlottingRemovalType Enum=Unparent
   ) {is=title K2Node_FunctionEntry}
   > n2 = Branch(Condition="Equal (GameplayTag)"(A=@n1, B=Get ItemInfoSlotTag()))
   > n3 = RemoveChild(self=Get HBox_ExtensionWidgets(), Content=Get ContentWidget())
   > Return Node!(execute, ReturnValue:bool=@n3)
   Return Node(execute=@n2.else, ReturnValue=false)
  ]}
 SetHeightOverride {nodes [
   SetHeightOverride() {is=title K2Node_FunctionEntry}
   > n1 = Branch(Condition="Equal Exactly (String)"(A=GetCurrentLanguage!() -> (ReturnValue:String), B=ko))
   n2 = Get SizeBox_Truncation() -> (SizeBox_Truncation:Size Box Object Reference)
   > ClearMinDesiredHeight(self:Size Box Object Reference=@n2)
   > ClearMaxDesiredHeight(self:Size Box Object Reference=@n2)
   > Set MinDesiredHeight!(execute, MinDesiredHeight:float=42, self:Size Box Object Reference=@n2) -> (then, Output_Get:float)
   > Set MaxDesiredHeight!(execute, MaxDesiredHeight:float=42, self:Size Box Object Reference=@n2) -> (then, Output_Get:float)
   n3 = Get SizeBox_Truncation()
   Set MinDesiredHeight(execute=@n1.else, MinDesiredHeight=38, self=@n3)
   > Set MaxDesiredHeight(MaxDesiredHeight=38, self=@n3)
  ]}
 UnregisterWidgetSlots {nodes [
   UnregisterWidgetSlots() {is=title K2Node_FunctionEntry}
   > UnregisterWidgetHost!(execute, WidgetHost:Widget Object Reference, WidgetHostTag:"Gameplay Tag Structure (by ref)"=Get ItemInfoSlotTag()) -> (then)
  ]}
]
variables [
 ItemInfoSlotTag {type=GameplayTag, default {tagName=DynamicUI.HUD.Fort.Athena.EquippedItemInfo.Slot}}
 Tag Container {type=GameplayTagContainer, default {gameplayTags [{tagName=Weapon.Ranged.Bow}]}}
]
widgets {
 info {
  inheritedWidgetCount=22
  parentClass=/Script/FortniteUI.BacchusEquippedItemReloadWidgetBase
  rootWidgetClass=&ScaleBox
  widgetCount=39
 }
 widgets [
  ScaleBox_HUDWidgetScale {
   widgetClassPath=&ScaleBox
   bInherited=true
   widget=ScaleBox_HUDWidgetScale
   properties {stretch=UserSpecified, visibility=HitTestInvisible}
  }
  SizeBox_2 {
   widgetClassPath=&SizeBox
   parent=ScaleBox_HUDWidgetScale
   widget=SizeBox_2
   properties {bOverride_HeightOverride=true, heightOverride=105}
   slot {class=ScaleBoxSlot, properties {horizontalAlignment=HAlign_Right, verticalAlignment=VAlign_Fill}}
  }
  VerticalBox_0 {widgetClassPath=&VerticalBox, parent=SizeBox_2, widget=VerticalBox_0, slot {class=SizeBoxSlot}}
  SizeBox_1 {
   widgetClassPath=&SizeBox
   parent=VerticalBox_0
   widget=SizeBox_1
   properties {bOverride_HeightOverride=true, heightOverride=76}
   slot {class=VerticalBoxSlot, properties {horizontalAlignment=HAlign_Right}}
  }
  Overlay_28 {widgetClassPath=&Overlay, parent=SizeBox_1, widget=Overlay_28, slot {class=SizeBoxSlot}}
  Switcher_Content {
   widgetClassPath=&CommonVisibilitySwitcher
   bInherited=true
   parent=Overlay_28
   widget=Switcher_Content
   slot {class=OverlaySlot, properties {horizontalAlignment=HAlign_Center, verticalAlignment=VAlign_Center}}
  }
  Panel_AmmoContent {
   widgetClassPath=&HorizontalBox
   bInherited=true
   parent=Switcher_Content
   widget=Panel_AmmoContent
   slot {class=CommonVisibilitySwitcherSlot}
  }
  SizeBox_0 {
   widgetClassPath=&SizeBox
   parent=Panel_AmmoContent
   widget=SizeBox_0
   properties {bOverride_MinDesiredWidth=true, minDesiredWidth=52}
   slot {class=HorizontalBoxSlot, properties {verticalAlignment=VAlign_Center, padding=<ltrb 8 0 0 0>, size {sizeRule=Automatic, value=0.75}}}
  }
  Text_MagazineAmmoCount {
   widgetClassPath=&CommonTextBlock
   bInherited=true
   bIsVariable=true
   parent=SizeBox_0
   widget=Text_MagazineAmmoCount
   properties {
    justification=Right
    style=/FortUILibrary/Styles/Typography/Foundation/TextStyle-Base-L-S.TextStyle-Base-L-S_C
    text=0
    font {
     bForceMonospaced=false
     bMaterialIsStencil=false
     fontMaterial=null
     fontObject=&NotoSans
     letterSpacing=0
     monospacedWidth=1
     size=32
     skewAmount=0
     typefaceFontName=NotoSans-Regular
     outlineSettings {
      bApplyOutlineToDropShadows=true
      bMiteredCorners=false
      bSeparateFillAlpha=true
      outlineMaterial=&M-FontShadow
      outlineSize=3
      outlineColor=<rgba 0 0 0 0.5>
     }
    }
    shadowOffset=<0 0>
   }
   slot {class=SizeBoxSlot}
  }
  Split1 {
   widgetClassPath=&Image
   bInherited=true
   bIsVariable=true
   parent=Panel_AmmoContent
   widget=Split1
   properties {
    brush {
     bIsDynamicallyLoaded=false
     drawAs=Image
     imageType=NoImage
     mirroring=NoMirror
     resourceName=None
     resourceObject=null
     tiling=NoTile
     imageSize=<4 40>
     margin=<ltrb 0 0 0 0>
     outlineSettings {
      bUseBrushTransparency=false
      roundingType=HalfHeightRadius
      width=0
      color {colorUseRule=UseColor_Specified, specifiedColor=<rgba 0 0 0 0>}
      cornerRadii=<0 0 0 0>
     }
     tintColor {colorUseRule=UseColor_Specified, specifiedColor=<rgba 1 1 1 1>}
     uVRegion {bIsValid=false, max=<0 0>, min=<0 0>}
    }
    colorAndOpacity=<rgba 1 1 1 0.2>
   }
   slot {class=HorizontalBoxSlot, properties {verticalAlignment=VAlign_Center}}
  }
  Text_SurplusAmmoCount {
   widgetClassPath=&CommonTextBlock
   bInherited=true
   bIsVariable=true
   parent=Panel_AmmoContent
   widget=Text_SurplusAmmoCount
   properties {
    style=&TextStyle-Base-L-S_C
    text=0
    font {
     bForceMonospaced=false
     bMaterialIsStencil=false
     fontMaterial=null
     fontObject=&NotoSans
     letterSpacing=0
     monospacedWidth=1
     size=32
     skewAmount=0
     typefaceFontName=NotoSans-Regular
     outlineSettings {
      bApplyOutlineToDropShadows=true
      bMiteredCorners=false
      bSeparateFillAlpha=true
      outlineMaterial=&M-FontShadow
      outlineSize=3
      outlineColor=<rgba 0 0 0 0.5>
     }
    }
    shadowOffset=<0 0>
   }
   slot {class=HorizontalBoxSlot, properties {horizontalAlignment=HAlign_Left, verticalAlignment=VAlign_Center}}
  }
  Overlay_0 {
   widgetClassPath=&Overlay
   parent=Panel_AmmoContent
   widget=Overlay_0
   slot {class=HorizontalBoxSlot, properties {horizontalAlignment=HAlign_Left, verticalAlignment=VAlign_Center, padding=<ltrb 4 0 0 0>}}
  }
  Image_Background {
   widgetClassPath=&Image
   bInherited=true
   parent=Overlay_0
   widget=Image_Background
   properties {
    visibility=Collapsed
    brush {
     bIsDynamicallyLoaded=false
     drawAs=Image
     imageType=FullColor
     mirroring=NoMirror
     resourceName=None
     resourceObject=/Game/Athena/HUD/Bacchus/Material/MI_UI_ButtonBG_Dark
     tiling=NoTile
     imageSize=<76 76>
     margin=<ltrb 0 0 0 0>
     outlineSettings {
      bUseBrushTransparency=false
      roundingType=HalfHeightRadius
      width=0
      color {colorUseRule=UseColor_Specified, specifiedColor=<rgba 0 0 0 0>}
      cornerRadii=<0 0 0 0>
     }
     tintColor {colorUseRule=UseColor_Specified, specifiedColor=<rgba 1 1 1 1>}
     uVRegion {bIsValid=false, max=<0 0>, min=<0 0>}
    }
   }
   slot {class=OverlaySlot, properties {horizontalAlignment=HAlign_Center, verticalAlignment=VAlign_Center}}
  }
  Image_AmmoIcon {
   widgetClassPath=&Image
   bInherited=true
   bIsVariable=true
   parent=Overlay_0
   widget=Image_AmmoIcon
   properties {
    visibility=SelfHitTestInvisible
    brush {
     bIsDynamicallyLoaded=false
     drawAs=Image
     imageType=NoImage
     mirroring=NoMirror
     resourceName=None
     resourceObject=/Game/UI/Foundation/Textures/Icons/Ammo/T-Icon-Ammo-Medium-64
     tiling=NoTile
     imageSize=<40 40>
     margin=<ltrb 0 0 0 0>
     outlineSettings {
      bUseBrushTransparency=false
      roundingType=HalfHeightRadius
      width=0
      color {colorUseRule=UseColor_Specified, specifiedColor=<rgba 0 0 0 0>}
      cornerRadii=<0 0 0 0>
     }
     tintColor {colorUseRule=UseColor_Specified, specifiedColor=<rgba 1 1 1 1>}
     uVRegion {bIsValid=false, max=<0 0>, min=<0 0>}
    }
   }
   slot {class=OverlaySlot, properties {horizontalAlignment=HAlign_Center, verticalAlignment=VAlign_Center}}
  }
  Panel_ResourceContent {
   widgetClassPath=&HorizontalBox
   bInherited=true
   parent=Switcher_Content
   widget=Panel_ResourceContent
   properties {renderTransform {angle=0, scale=<1 1>, shear=<0 0>, translation=<0 -140>}}
   slot {class=CommonVisibilitySwitcherSlot}
  }
  Text_ResourceCount {
   widgetClassPath=&CommonTextBlock
   bInherited=true
   bIsVariable=true
   parent=Panel_ResourceContent
   widget=Text_ResourceCount
   properties {
    style=&TextStyle-Base-L-S_C
    text=999
    font {
     bForceMonospaced=false
     bMaterialIsStencil=false
     fontMaterial=null
     fontObject=&NotoSans
     letterSpacing=0
     monospacedWidth=1
     size=32
     skewAmount=0
     typefaceFontName=NotoSans-Regular
     outlineSettings {
      bApplyOutlineToDropShadows=true
      bMiteredCorners=false
      bSeparateFillAlpha=true
      outlineMaterial=&M-FontShadow
      outlineSize=3
      outlineColor=<rgba 0 0 0 0.5>
     }
    }
    shadowOffset=<0 0>
   }
   slot {class=HorizontalBoxSlot, properties {horizontalAlignment=HAlign_Left, verticalAlignment=VAlign_Center, padding=<ltrb 4 0 0 0>}}
  }
  Panel_UtilityContent {
   widgetClassPath=&HorizontalBox
   bInherited=true
   parent=Switcher_Content
   widget=Panel_UtilityContent
   slot {class=CommonVisibilitySwitcherSlot}
  }
  Image_UtilityIcon {
   widgetClassPath=&Image
   bInherited=true
   bIsVariable=true
   parent=Panel_UtilityContent
   widget=Image_UtilityIcon
   properties {
    visibility=SelfHitTestInvisible
    brush {
     bIsDynamicallyLoaded=false
     drawAs=Image
     imageType=NoImage
     mirroring=NoMirror
     resourceName=None
     resourceObject=/Game/UI/Foundation/Textures/Icons/Items/T-Icon-S-Loot-Stone
     tiling=NoTile
     imageSize=<40 40>
     margin=<ltrb 0 0 0 0>
     outlineSettings {
      bUseBrushTransparency=false
      roundingType=HalfHeightRadius
      width=0
      color {colorUseRule=UseColor_Specified, specifiedColor=<rgba 0 0 0 0>}
      cornerRadii=<0 0 0 0>
     }
     tintColor {colorUseRule=UseColor_Specified, specifiedColor=<rgba 1 1 1 1>}
     uVRegion {bIsValid=false, max=<0 0>, min=<0 0>}
    }
   }
   slot {class=HorizontalBoxSlot, properties {verticalAlignment=VAlign_Center}}
  }
  Text_UtilityAmmoCount {
   widgetClassPath=&CommonTextBlock
   bInherited=true
   parent=Panel_UtilityContent
   widget=Text_UtilityAmmoCount
   properties {
    style=&TextStyle-Base-L-S_C
    text=0
    font {
     bForceMonospaced=false
     bMaterialIsStencil=false
     fontMaterial=null
     fontObject=&NotoSans
     letterSpacing=0
     monospacedWidth=1
     size=32
     skewAmount=0
     typefaceFontName=NotoSans-Regular
     outlineSettings {
      bApplyOutlineToDropShadows=true
      bMiteredCorners=false
      bSeparateFillAlpha=true
      outlineMaterial=&M-FontShadow
      outlineSize=3
      outlineColor=<rgba 0 0 0 0.5>
     }
    }
    shadowOffset=<0 0>
   }
   slot {
    class=HorizontalBoxSlot
    properties {horizontalAlignment=HAlign_Right, verticalAlignment=VAlign_Center, padding=<ltrb 8 0 0 0>, size {sizeRule=Automatic, value=0.75}}
   }
  }
  Image_UtilitySplit {
   widgetClassPath=&Image
   bInherited=true
   bIsVariable=true
   parent=Panel_UtilityContent
   widget=Image_UtilitySplit
   properties {
    visibility=Collapsed
    brush {
     bIsDynamicallyLoaded=false
     drawAs=Image
     imageType=NoImage
     mirroring=NoMirror
     resourceName=None
     resourceObject=/Game/Athena/HUD/Art/Materials/M_UI_HUD_AmmoCount_DividerLine
     tiling=NoTile
     imageSize=<4 22>
     margin=<ltrb 0 0 0 0>
     outlineSettings {
      bUseBrushTransparency=false
      roundingType=HalfHeightRadius
      width=0
      color {colorUseRule=UseColor_Specified, specifiedColor=<rgba 0 0 0 0>}
      cornerRadii=<0 0 0 0>
     }
     tintColor {colorUseRule=UseColor_Specified, specifiedColor=<rgba 1 1 1 1>}
     uVRegion {bIsValid=false, max=<0 0>, min=<0 0>}
    }
    renderTransform {angle=0, scale=<1 1>, shear=<-12 0>, translation=<0 3>}
   }
   slot {class=HorizontalBoxSlot, properties {verticalAlignment=VAlign_Center, padding=<ltrb 4 0 4 0>}}
  }
  Text_SurplusUtilityAmmoCount {
   widgetClassPath=&CommonTextBlock
   bInherited=true
   bIsVariable=true
   parent=Panel_UtilityContent
   widget=Text_SurplusUtilityAmmoCount
   properties {
    text=0
    visibility=Collapsed
    font {
     bForceMonospaced=false
     bMaterialIsStencil=false
     fontMaterial=null
     fontObject=&Burbank
     letterSpacing=0
     monospacedWidth=1
     size=20
     skewAmount=0
     typefaceFontName=Burbank-PlayerPower
     outlineSettings {
      bApplyOutlineToDropShadows=true
      bMiteredCorners=false
      bSeparateFillAlpha=true
      outlineMaterial=&M-FontShadow
      outlineSize=3
      outlineColor=<rgba 0 0 0 0.5>
     }
    }
    renderTransform {angle=0, scale=<1 1>, shear=<-12 0>, translation=<0 3>}
    shadowOffset=<0 0>
   }
   slot {class=HorizontalBoxSlot, properties {horizontalAlignment=HAlign_Left, verticalAlignment=VAlign_Center}}
  }
  Overlay_UtilityAmmoIcon {
   widgetClassPath=&Overlay
   bInherited=true
   bIsVariable=true
   parent=Panel_UtilityContent
   widget=Overlay_UtilityAmmoIcon
   properties {visibility=Collapsed}
   slot {class=HorizontalBoxSlot, properties {horizontalAlignment=HAlign_Left, verticalAlignment=VAlign_Center, padding=<ltrb 4 0 0 0>}}
  }
  Image_UtilityBackground {
   widgetClassPath=&Image
   parent=Overlay_UtilityAmmoIcon
   widget=Image_UtilityBackground
   properties {
    visibility=Collapsed
    brush {
     bIsDynamicallyLoaded=false
     drawAs=Image
     imageType=FullColor
     mirroring=NoMirror
     resourceName=None
     resourceObject=&MI_UI_ButtonBG_Dark
     tiling=NoTile
     imageSize=<76 76>
     margin=<ltrb 0 0 0 0>
     outlineSettings {
      bUseBrushTransparency=false
      roundingType=HalfHeightRadius
      width=0
      color {colorUseRule=UseColor_Specified, specifiedColor=<rgba 0 0 0 0>}
      cornerRadii=<0 0 0 0>
     }
     tintColor {colorUseRule=UseColor_Specified, specifiedColor=<rgba 1 1 1 1>}
     uVRegion {bIsValid=false, max=<0 0>, min=<0 0>}
    }
   }
   slot {class=OverlaySlot, properties {horizontalAlignment=HAlign_Center, verticalAlignment=VAlign_Center}}
  }
  Image_UtilityAmmoIcon {
   widgetClassPath=&Image
   bIsVariable=true
   parent=Overlay_UtilityAmmoIcon
   widget=Image_UtilityAmmoIcon
   properties {
    visibility=SelfHitTestInvisible
    brush {
     bIsDynamicallyLoaded=false
     drawAs=Image
     imageType=NoImage
     mirroring=NoMirror
     resourceName=None
     resourceObject=&T-Icon-Ammo-Medium-64
     tiling=NoTile
     imageSize=<40 40>
     margin=<ltrb 0 0 0 0>
     outlineSettings {
      bUseBrushTransparency=false
      roundingType=HalfHeightRadius
      width=0
      color {colorUseRule=UseColor_Specified, specifiedColor=<rgba 0 0 0 0>}
      cornerRadii=<0 0 0 0>
     }
     tintColor {colorUseRule=UseColor_Specified, specifiedColor=<rgba 1 1 1 1>}
     uVRegion {bIsValid=false, max=<0 0>, min=<0 0>}
    }
   }
   slot {class=OverlaySlot, properties {horizontalAlignment=HAlign_Center, verticalAlignment=VAlign_Center}}
  }
  Text_OutOfAmmo {
   widgetClassPath=&CommonTextBlock
   bInherited=true
   bIsVariable=true
   parent=Overlay_28
   widget=Text_OutOfAmmo
   properties {
    text="Out of Ammo!"
    visibility=Collapsed
    font {
     bForceMonospaced=false
     bMaterialIsStencil=false
     fontMaterial=null
     fontObject=&NotoSans
     letterSpacing=0
     monospacedWidth=1
     size=18
     skewAmount=0
     typefaceFontName=NotoSans-Bold
     outlineSettings {
      bApplyOutlineToDropShadows=true
      bMiteredCorners=false
      bSeparateFillAlpha=false
      outlineMaterial=null
      outlineSize=1
      outlineColor=<rgba 0 0 0 0.498>
     }
    }
    renderTransform {angle=-5, scale=<1 1>, shear=<0 0>, translation=<0 0>}
    shadowOffset=<0 0>
   }
   slot {class=OverlaySlot, properties {horizontalAlignment=HAlign_Center, verticalAlignment=VAlign_Center}}
  }
  TutorialHighlight {
   widgetClassPath=&TutorialHighlight
   bIsVariable=true
   parent=Overlay_28
   widget=TutorialHighlight
   properties {customHighlightWidget=/Game/Athena/HUD/Tutorial/Tutorial_Outline_Square.Tutorial_Outline_Square_C}
   slot {class=OverlaySlot, properties {horizontalAlignment=HAlign_Fill, verticalAlignment=VAlign_Fill}}
  }
  SizeBox_43 {
   widgetClassPath=&SizeBox
   parent=Overlay_28
   widget=SizeBox_43
   properties {bOverride_HeightOverride=true, heightOverride=60}
   slot {class=OverlaySlot, properties {horizontalAlignment=HAlign_Fill, padding=<ltrb -15 -60 15 0>}}
  }
  FortPlaylistUIExtensionSlot_AboveAmmoCount {
   widgetClassPath=/Script/FortniteUI.FortPlaylistUIExtensionSlot
   bIsVariable=true
   parent=SizeBox_43
   widget=FortPlaylistUIExtensionSlot_AboveAmmoCount
   properties {entryWidgetClass=&CommonUserWidget, extensionSlots [AboveBacchusEquippedItemReloadWidget]}
   slot {class=SizeBoxSlot}
  }
  SwitchPopUp {
   widgetClassPath=&VerticalBox
   bIsVariable=true
   parent=VerticalBox_0
   widget=SwitchPopUp
   properties {visibility=Hidden}
   slot {class=VerticalBoxSlot, properties {horizontalAlignment=HAlign_Right, verticalAlignment=VAlign_Top}}
  }
  AthenaEquippedItemDurabilityWidget {
   widgetClassPath=/Game/Athena/HUD/AthenaEquippedItemDurabilityWidget
   parent=SwitchPopUp
   widget=AthenaEquippedItemDurabilityWidget
   slot {class=VerticalBoxSlot, properties {horizontalAlignment=HAlign_Right}}
  }
  VBox_ItemName {widgetClassPath=&VerticalBox, parent=SwitchPopUp, widget=VBox_ItemName, slot {class=VerticalBoxSlot}}
  Overlay_2 {
   widgetClassPath=&Overlay
   parent=VBox_ItemName
   widget=Overlay_2
   properties {clipping=ClipToBounds}
   slot {class=VerticalBoxSlot, properties {verticalAlignment=VAlign_Center}}
  }
  SizeBox_Truncation {
   widgetClassPath=&SizeBox
   bIsVariable=true
   parent=Overlay_2
   widget=SizeBox_Truncation
   properties {
    bOverride_MaxDesiredWidth=true
    bOverride_MinDesiredWidth=true
    heightOverride=38
    maxDesiredHeight=38
    maxDesiredWidth=250
    minDesiredHeight=38
    minDesiredWidth=250
   }
   slot {class=OverlaySlot, properties {horizontalAlignment=HAlign_Right, verticalAlignment=VAlign_Fill, padding=<ltrb 0 0 0 8>}}
  }
  SwitchPopUpText {
   widgetClassPath=&CommonTextBlock
   bIsVariable=true
   parent=SizeBox_Truncation
   widget=SwitchPopUpText
   properties {
    bIsScrollingEnabled=false
    justification=Right
    lineHeightPercentage=0.8
    text="[PH] Item Name"
    textOverflowPolicy=Ellipsis
    textTransformPolicy=ToUpper
    wrapTextAt=230
    font {
     bForceMonospaced=false
     bMaterialIsStencil=false
     fontMaterial=null
     fontObject=&Burbank
     letterSpacing=0
     monospacedWidth=1
     size=16.5
     skewAmount=0
     typefaceFontName=BurbankRegular-Black
     outlineSettings {
      bApplyOutlineToDropShadows=true
      bMiteredCorners=false
      bSeparateFillAlpha=true
      outlineMaterial=&M-FontShadow
      outlineSize=2
      outlineColor=<rgba 0 0 0 0.5>
     }
    }
    margin=<ltrb 0 0 0 4>
    shadowOffset=<0 0>
   }
   slot {class=SizeBoxSlot}
  }
  HBox_ExtensionWidgets {
   widgetClassPath=&HorizontalBox
   bIsVariable=true
   parent=VBox_ItemName
   widget=HBox_ExtensionWidgets
   properties {renderTransform {angle=0, scale=<1 1>, shear=<0 0>, translation=<0 6>}}
   slot {class=VerticalBoxSlot, properties {horizontalAlignment=HAlign_Right}}
  }
  Image_AmmoSplit {widgetClassPath=&Image, bInherited=true, widget=null}
  Image_ResourceIcon {widgetClassPath=&Image, bInherited=true, widget=null}
  Image_RechargeableIcon {widgetClassPath=&Image, bInherited=true, widget=null}
  RootScaleBox {widgetClassPath=&ScaleBox, bInherited=true, widget=null}
 ]
}
